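# Physics asset for the Punchy back bling cosmetic, setting up bodies for the pack, two dangling gloves and their left and right ribbons.
# A box covers the pack and capsules cover the other bones, with pack and glove bases joined to root by angularly free constraints.
# Front ribbon chains start at the gloves and back chains at the pack, with swing and twist limits.
# Angularly locked constraints join each side's ribbon tips, closing a loop.
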PhysicsAsset /BRCosmetics/Accessories/FORT_Backpacks/F_MED_Punchy_Backpack/Meshes/F_MED_Punchy_Pack_Physics
bodies {
 dyn_l_glove_01 [{shapeType=Capsule, length=1, radius=1}]
 dyn_l_glove_02 [{shapeType=Capsule, length=3.2103, radius=3.2897, center=<3.5 0 0>, rotation=<rot 90 0 0>}]
 dyn_l_ribbon_bk_01 [{shapeType=Capsule, length=3.75, radius=1.5, center=<2 0 0>, rotation=<rot 90 0 0>}]
 dyn_l_ribbon_bk_02 [{shapeType=Capsule, length=3.75, radius=1.5, center=<1.6607 -0.0001 -0.9323>, rotation=<rot 69.9998 0 0>}]
 dyn_l_ribbon_bk_03 [{shapeType=Capsule, length=3.75, radius=1.5, center=<2 0 0>, rotation=<rot 90 0 0>}]
 dyn_l_ribbon_fn_01 [{shapeType=Capsule, length=2, radius=1.5, center=<1.5 0 0>, rotation=<rot 90 0 0>}]
 dyn_l_ribbon_fn_02 [{shapeType=Capsule, length=4.5, radius=1.5, center=<2 0 0>, rotation=<rot 90 0 0>}]
 dyn_l_ribbon_fn_03 [{shapeType=Capsule, length=2.5, radius=1.5, rotation=<rot 90 0 0>}]
 dyn_r_glove_01 [{shapeType=Capsule, length=1, radius=1}]
 dyn_r_glove_02 [{shapeType=Capsule, length=3.2103, radius=3.2897, center=<3.5 0 0>, rotation=<rot 90 0 0>}]
 dyn_r_ribbon_bk_01 [{shapeType=Capsule, length=3.75, radius=1.5, center=<2 0 0>, rotation=<rot 90 0 0>}]
 dyn_r_ribbon_bk_02 [{shapeType=Capsule, length=3.75, radius=1.5, center=<1.6607 -0.0001 -0.9323>, rotation=<rot 69.9998 0 0>}]
 dyn_r_ribbon_bk_03 [{shapeType=Capsule, length=3.75, radius=1.5, center=<2 0 0>, rotation=<rot 90 0 0>}]
 dyn_r_ribbon_fn_01 [{shapeType=Capsule, length=2, radius=1.5, center=<1.5 0 0>, rotation=<rot 90 0 0>}]
 dyn_r_ribbon_fn_02 [{shapeType=Capsule, length=4.5, radius=1.5, center=<2 0 0>, rotation=<rot 90 0 0>}]
 dyn_r_ribbon_fn_03 [{shapeType=Capsule, length=2.5, radius=1.5, rotation=<rot 90 0 0>}]
 pack [{shapeType=Box, extentX=22.821, extentY=9.2794, extentZ=21.1425, center=<0 -2.9878 -8.8149>}]
 root [{shapeType=Capsule, length=21.1109, radius=0.9032, shapeName=root_capsule, center=<0 -8.0236 -19.2189>, rotation=<rot 90 0 0>}]
}
constraints [
 {bone1Name=dyn_l_ribbon_fn_01, bone2Name=dyn_l_glove_01, swing1LimitDegrees=45, swing2LimitDegrees=45, twistLimitDegrees=2}
 {bone1Name=dyn_l_ribbon_fn_02, bone2Name=dyn_l_ribbon_fn_01, swing1LimitDegrees=45, swing2LimitDegrees=45, twistLimitDegrees=2}
 {bone1Name=dyn_l_ribbon_fn_03, bone2Name=dyn_l_ribbon_fn_02, swing1LimitDegrees=45, swing2LimitDegrees=45, twistLimitDegrees=10}
 {bone1Name=dyn_l_ribbon_bk_01, bone2Name=pack, swing1LimitDegrees=45, swing2LimitDegrees=45, twistLimitDegrees=2}
 {bone1Name=dyn_l_ribbon_bk_02, bone2Name=dyn_l_ribbon_bk_01, swing1LimitDegrees=45, swing2LimitDegrees=45, twistLimitDegrees=5}
 {bone1Name=dyn_l_ribbon_bk_03, bone2Name=dyn_l_ribbon_bk_02, swing1LimitDegrees=45, swing2LimitDegrees=45, twistLimitDegrees=10}
 {bone1Name=dyn_r_ribbon_fn_01, bone2Name=dyn_r_glove_01, swing1LimitDegrees=45, swing2LimitDegrees=45, twistLimitDegrees=2}
 {bone1Name=dyn_r_ribbon_fn_02, bone2Name=dyn_r_ribbon_fn_01, swing1LimitDegrees=45, swing2LimitDegrees=45, twistLimitDegrees=5}
 {bone1Name=dyn_r_ribbon_fn_03, bone2Name=dyn_r_ribbon_fn_02, swing1LimitDegrees=45, swing2LimitDegrees=45, twistLimitDegrees=10}
 {bone1Name=dyn_r_ribbon_bk_01, bone2Name=pack, swing1LimitDegrees=45, swing2LimitDegrees=45, twistLimitDegrees=10}
 {bone1Name=dyn_r_ribbon_bk_02, bone2Name=dyn_r_ribbon_bk_01, swing1LimitDegrees=45, swing2LimitDegrees=45, twistLimitDegrees=5}
 {bone1Name=dyn_r_ribbon_bk_03, bone2Name=dyn_r_ribbon_bk_02, swing1LimitDegrees=45, swing2LimitDegrees=45, twistLimitDegrees=10}
 {
  bone1Name=dyn_r_ribbon_fn_03
  bone2Name=dyn_r_ribbon_bk_03
  swing1LimitDegrees=30
  swing1Motion=Locked
  swing2LimitDegrees=30
  swing2Motion=Locked
  twistLimitDegrees=2
  twistMotion=Locked
 }
 {bone1Name=dyn_r_glove_02, bone2Name=dyn_r_glove_01, swing1LimitDegrees=60, swing2LimitDegrees=60, twistLimitDegrees=5}
 {bone1Name=dyn_l_glove_02, bone2Name=dyn_l_glove_01, swing1LimitDegrees=60, swing2LimitDegrees=60, twistLimitDegrees=5}
 {
  bone1Name=pack
  bone2Name=root
  swing1LimitDegrees=45
  swing1Motion=Free
  swing2LimitDegrees=45
  swing2Motion=Free
  twistLimitDegrees=45
  twistMotion=Free
 }
 {
  bone1Name=dyn_l_glove_01
  bone2Name=root
  swing1LimitDegrees=45
  swing1Motion=Free
  swing2LimitDegrees=45
  swing2Motion=Free
  twistLimitDegrees=45
  twistMotion=Free
 }
 {
  bone1Name=dyn_r_glove_01
  bone2Name=root
  swing1LimitDegrees=45
  swing1Motion=Free
  swing2LimitDegrees=45
  swing2Motion=Free
  twistLimitDegrees=45
  twistMotion=Free
 }
 {
  bone1Name=dyn_l_ribbon_fn_03
  bone2Name=dyn_l_ribbon_bk_03
  swing1LimitDegrees=45
  swing1Motion=Locked
  swing2LimitDegrees=45
  swing2Motion=Locked
  twistLimitDegrees=45
  twistMotion=Locked
 }
]
