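# Animation Blueprint for the Punchy backpack cosmetic that adds secondary physics motion to the pack and its ribbon bones.
# The cached input pose feeds two branches blended on bRigidBodyEnabled.
# When true, a RigidBody simulation feeds an additive roll on the pack bone.
# Otherwise, AnimDynamics chains drive the front ribbons, followed by trail controllers on front and back ribbons, which the event graph switches off during skydiving.
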AnimBlueprint /BRCosmetics/Accessories/FORT_Backpacks/F_MED_Punchy_Backpack/Meshes/F_MED_Punchy_Pack_AnimBlueprint
target_skeleton=&/F_MED_Punchy_Pack_Skeleton
graphs [
 AnimGraph {nodes [
   n1 = RigidBody(ComponentPose=Local To Component(LocalPose=Use cached pose 'In Pose'())) {
    bClampLinearTranslationLimitToRefPose=true
    bOverrideWorldGravity=true
    overridePhysicsAsset=&/F_MED_Punchy_Pack_Physics
    simulationSpace=BaseBoneSpace
    alphaBoolBlend {blendOption=Linear}
    baseBoneRef {boneName=root}
    componentAppliedLinearAccClamp=<1000 1000 1000>
    componentLinearAccScale=<0.35 0.35 0.35>
    componentLinearVelScale=<0.1 0.1 0.1>
    overrideWorldGravity {z=-600}
   }
   n2 = "Transform (Modify) Bone - Bone: pack"(ComponentPose=@n1, Translation="0,0,0", Rotation="0,0,7") {rotationMode=BMM_Additive, alphaBoolBlend {blendOption=Linear}, boneToModify {boneName=pack}, rotation {roll=7}}
   n3 = "Anim Dynamics - Chain: dyn_l_ribbon_fn_01 -> dyn_l_ribbon_fn_03"(ComponentPose=Local To Component(LocalPose=Use cached pose 'In Pose'()), Alpha=1, ExternalForce="0,0,0") {
    angularDampingOverride=0.99
    angularSpringConstant=20000
    bAngularSpring=true
    bChain=true
    bOverrideAngularDamping=true
    bUseGravityOverride=true
    simulationSpace=RootRelative
    alphaBoolBlend {blendOption=Linear}
    boundBone {boneName=dyn_l_ribbon_fn_01}
    chainEnd {boneName=dyn_l_ribbon_fn_03}
    componentAppliedLinearAccClamp=<1500 1500 1500>
    componentLinearAccScale=<0.3 0.3 0.3>
    componentLinearVelScale=<3 0.3 0.3>
    gravityOverride {z=-980}
    physicsBodyDefinitions [
     {
      boundBone {boneName=dyn_l_ribbon_fn_01}
      boxExtents=<30 30 30>
      constraintSetup {angularLimitsMax=<5 30 30>, angularLimitsMin=<-5 -30 -30>, angularTarget=<0 0 0>}
      localJointOffset=<-0.5 0 0>
     }
     {
      boundBone {boneName=dyn_l_ribbon_fn_02}
      boxExtents=<30 30 30>
      constraintSetup {angularLimitsMax=<5 30 30>, angularLimitsMin=<-5 -30 -30>, angularTarget=<0 0 0>}
      localJointOffset=<-0.5 0 0>
     }
     {
      boundBone {boneName=dyn_l_ribbon_fn_03}
      boxExtents=<30 30 30>
      constraintSetup {angularLimitsMax=<5 30 30>, angularLimitsMin=<-5 -30 -30>, angularTarget=<0 0 0>}
      localJointOffset=<-0.5 0 0>
     }
    ]
    planarLimits [{drivingBone {boneName=None}, planeTransform {location=<0 -6 0>, rotation=<rot 0 0 -80>}}]
   }
   n4 = "Anim Dynamics - Chain: dyn_r_ribbon_fn_01 -> dyn_r_ribbon_fn_03"(ComponentPose=@n3, Alpha=1, ExternalForce="0,0,0") {
    angularDampingOverride=0.99
    angularSpringConstant=20000
    bAngularSpring=true
    bChain=true
    bOverrideAngularDamping=true
    bUseGravityOverride=true
    simulationSpace=RootRelative
    alphaBoolBlend {blendOption=Linear}
    boundBone {boneName=dyn_r_ribbon_fn_01}
    chainEnd {boneName=dyn_r_ribbon_fn_03}
    componentAppliedLinearAccClamp=<1500 1500 1500>
    componentLinearAccScale=<0.3 0.3 0.3>
    componentLinearVelScale=<3 0.3 0.3>
    gravityOverride {z=-980}
    physicsBodyDefinitions [
     {
      boundBone {boneName=dyn_r_ribbon_fn_01}
      boxExtents=<30 1 30>
      constraintSetup {angularLimitsMax=<5 30 30>, angularLimitsMin=<-5 -30 -30>, angularTarget=<0 0 0>}
      localJointOffset=<-0.5 0 0>
     }
     {
      boundBone {boneName=dyn_r_ribbon_fn_02}
      boxExtents=<30 30 30>
      constraintSetup {angularLimitsMax=<5 30 30>, angularLimitsMin=<-5 -30 -30>, angularTarget=<0 0 0>}
      localJointOffset=<-0.5 0 0>
     }
     {
      boundBone {boneName=dyn_r_ribbon_fn_03}
      boxExtents=<30 30 30>
      constraintSetup {angularLimitsMax=<5 30 30>, angularLimitsMin=<-5 -30 -30>, angularTarget=<0 0 0>}
      localJointOffset=<-0.5 0 0>
     }
    ]
    planarLimits [{drivingBone {boneName=None}, planeTransform {location=<0 -6 0>, rotation=<rot 0 0 -80>}}]
   }
   n5 = Get NO_SkyDiving() -> (NO_SkyDiving:double)
   n6 = Trail controller - Bone: dyn_l_ribbon_fn_02(ComponentPose=@n4, Alpha=@n5) {
    bLimitRotation=true
    relaxationSpeedScale=2
    rotationLimits [{limitMax=<5 35 45>, limitMin=<-5 -35 -45>}, {limitMax=<5 35 45>, limitMin=<-5 -35 -45>}]
    rotationOffsets [<0 0 0>, <0 0 0>]
    trailBone {boneName=dyn_l_ribbon_fn_02}
   }
   n7 = Trail controller - Bone: dyn_r_ribbon_fn_02(ComponentPose=@n6, Alpha=@n5) {
    bLimitRotation=true
    relaxationSpeedScale=2
    rotationLimits [{limitMax=<5 35 45>, limitMin=<-5 -35 -45>}, {limitMax=<5 35 45>, limitMin=<-5 -35 -45>}]
    rotationOffsets [<0 0 0>, <0 0 0>]
    trailBone {boneName=dyn_r_ribbon_fn_02}
   }
   n8 = Trail controller - Bone: dyn_l_ribbon_bk_02(ComponentPose=@n7, Alpha=@n5) {
    bLimitRotation=true
    relaxationSpeedScale=2
    rotationLimits [{limitMax=<5 35 45>, limitMin=<-5 -35 -45>}, {limitMax=<5 35 45>, limitMin=<-5 -35 -45>}]
    rotationOffsets [<0 0 0>, <0 0 0>]
    trailBone {boneName=dyn_l_ribbon_bk_02}
   }
   n9 = Trail controller - Bone: dyn_r_ribbon_bk_02(ComponentPose=@n8, Alpha=@n5) {
    bLimitRotation=true
    relaxationSpeedScale=2
    rotationLimits [{limitMax=<5 35 45>, limitMin=<-5 -35 -45>}, {limitMax=<5 35 45>, limitMin=<-5 -35 -45>}]
    rotationOffsets [<0 0 0>, <0 0 0>]
    trailBone {boneName=dyn_r_ribbon_bk_02}
   }
   Output Pose(Result=Blend Poses by bool(BlendPose_0=Component To Local(ComponentPose=@n2), BlendPose_1=Component To Local(ComponentPose=@n9), bActiveValue=Get bRigidBodyEnabled()) {blendType=Linear}) {layerGroup=None}
   AnimGraphNode_SubInput_0 = Input Pose()
   Save cached pose 'In Pose'(Pose=@AnimGraphNode_SubInput_0) {cacheName=In Pose}
  ]}
 EventGraph {nodes [
   Event BlueprintUpdateAnimation()
   > Set NO_SkyDiving(NO_SkyDiving:double=SelectFloat(A=0, B=1, bPickA=Get bIsSkyDiving())) -> (Output_Get:double)
   TryGetPawnOwner()
  ]}
]
variables [NO_SkyDiving {type=number}]
